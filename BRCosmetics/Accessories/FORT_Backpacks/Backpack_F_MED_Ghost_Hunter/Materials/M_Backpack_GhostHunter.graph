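# Additive, two-sided material for the Ghost Hunter backpack cosmetic, drawing glowing left and right counters plus a core.
# Red and green vertex colors mask the sides, comparing against progress parameters to select pip color, glow, icon or dimmed square-pattern states.
# Blue vertex color drives core and overlay color; similar comparisons push vertices along normals.
# View-angle opacity softens edges before MF_TechArtOverrides_accessories; a rotate-about-axis branch is unconnected.
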Material /BRCosmetics/Accessories/FORT_Backpacks/Backpack_F_MED_Ghost_Hunter/Materials/M_Backpack_GhostHunter
graph {
 properties {
  bUseMaterialAttributes=true
  bUsedWithMeshDeformer=true
  bUsedWithSkeletalMesh=true
  blendMode=BLEND_Additive
  pixelDepthOffsetMode=PDOM_Legacy
  twoSided=true
  usedShadingModels=Default Lit
 }
 nodes [
  n1 = NamedRerouteDeclaration(None=VertexColor().R) {name=RedVertexColor, nodeColor=<rgba 1 0.4922 0 1>}
  n2 = NamedRerouteDeclaration(None=Ceil(None=@n1)) {name=RedVertexColor_Ceil, nodeColor=<rgba 0.1953 0 1 1>}
  n3 = ScalarParameter() {defaultValue=0.97, group=Functional, parameterName=LeftCounterProgress}
  n4 = VectorParameter() {parameterName=Color_LeftSide_Active, defaultValue=<rgba 0.0284 0 2 0>}
  n5 = TextureCoordinate() {coordinateIndex=1}
  n6 = Saturate(None=Constant() {r=0.5})
  n7 = Subtract(A=Saturate(None=ComponentMask(None=@n5) {r=true}), B=@n6)
  n8 = Multiply(A=@n7, B=Divide(B=Subtract(B=@n6)))
  n9 = Multiply(A=Multiply(A=Divide(B=@n6), B=@n7)) {constB=-1}
  n10 = Saturate(None=Power(Base=Max(A=@n8, B=@n9), Exp=ScalarParameter() {defaultValue=3, parameterName=Glow_Power}) {constExponent=5})
  n11 = VectorParameter() {parameterName=Color_GlowLeftSide, defaultValue=<rgba 0 0.5092 1 0>}
  n12 = Add(
   A=MaterialFunctionCall(R=@n4.R, G=@n4.R, B=@n4.R) {materialFunction=&AppendMany}.RGB
   B=Multiply(A=NamedRerouteUsage(), B=Multiply(A=@n10, B=@n11.RGB))
  )
  n13 = TextureCoordinate() {coordinateIndex=1}
  n14 = Constant2Vector() {g=0.65, r=1}
  n15 = TextureSample(UVs=MaterialFunctionCall(
    UVs=MaterialFunctionCall(UVs=@n13, Texture Scale=@n14) {materialFunction=&ScaleUVsByCenter}.UVs
    "Rotation Angle (0-1)"=Constant() {r=0.25}
   ) {materialFunction=&CustomRotator}.Rotated Values) {
   samplerSource=SSM_Clamp_WorldGroupSettings
   texture=/BRCosmetics/Accessories/FORT_Backpacks/Backpack_F_MED_Ghost_Hunter/FX/t_GhostHunter_BackPack_Icons
  }
  n16 = Add(A=@n12, B=@n15.G)
  n17 = VectorParameter() {parameterName=Color_Activated, defaultValue=<rgba 5 5 5 0>}
  n18 = Sine(None=Multiply(A=Time(), B=ScalarParameter() {parameterName=SCALAR_SquareWobble}))
  n19 = Constant() {r=-1}
  n20 = Constant() {r=1}
  n21 = Constant() {r=0.45}
  n22 = Constant() {r=0.35}
  n23 = Multiply(
   A=MaterialFunctionCall(R=@n17.RGB, G=@n17.RGB, B=@n17.RGB) {materialFunction=&AppendMany}.RGB
   B=MaterialFunctionCall!(
    UVs=TextureCoordinate() {coordinateIndex=1}
    Size=MaterialFunctionCall(Input=@n18, Input Low=@n19, Input High=@n20, Target Low=@n21, Target High=@n22) {materialFunction=&RemapValueRange}.Result
    SizeScale=Constant() {r=0.9}
   ) {materialFunction=/Game/Packages/Fortress_SharedMaterials/Base_Material_Functions/SquarePatternFromUVs}.PatternOutput
  )
  n24 = Add(A=@n23, B=@n12)
  n25 = ScalarParameter() {defaultValue=0.011, parameterName=SCALAR_Opacity}
  n26 = If(
   A=@n1
   B=NamedRerouteDeclaration(None=Add(A=@n3) {constB=0}) {name=LeftCounterProgress, nodeColor=<rgba 1 0 0.1406 1>}
   "A > B"=If(A=NamedRerouteUsage(), B=Constant() {r=0.97}, "A > B"=@n16, "A == B"=@n16, "A < B"=@n12) {equalsThreshold=0.01}
   "A == B"=LinearInterpolate(
    A=@n12
    B=VectorParameter() {parameterName=Color_ActivatedPipLeftSide, defaultValue=<rgba 80 61.6252 0 0>}.RGB
    Alpha=ScalarParameter() {defaultValue=1, parameterName=Color_LeftSideMultiplier}
   )
   "A < B"=Multiply(A=LinearInterpolate(A=@n24, B=@n12) {constAlpha=0}, B=@n25)
  ) {equalsThreshold=0.01}
  n27 = VectorParameter() {parameterName=Color_GlowRightSide, defaultValue=<rgba 2 0 1.8949 0>}
  n28 = VectorParameter() {parameterName=Color_RightSide, defaultValue=<rgba 0.7369 0 2 0>}
  n29 = Add(
   A=Multiply(A=NamedRerouteUsage(), B=Multiply(A=@n10, B=@n27.RGB))
   B=MaterialFunctionCall(R=@n28.R, G=@n28.R, B=@n28.R) {materialFunction=&AppendMany}.RGB
  )
  n30 = Add(A=@n29, B=@n15.R)
  n31 = If(
   A=NamedRerouteUsage()
   B=NamedRerouteUsage()
   "A > B"=If(A=NamedRerouteUsage(), B=Constant() {r=0.97}, "A > B"=@n30, "A == B"=@n30, "A < B"=@n29) {equalsThreshold=0.01}
   "A == B"=LinearInterpolate(
    A=@n29
    B=VectorParameter() {parameterName=Color_ActivatedPipRightSide, defaultValue=<rgba 80 61.6252 0 0>}.RGB
    Alpha=ScalarParameter() {defaultValue=1, parameterName=Color_RightSideMultiplier}
   )
   "A < B"=Multiply(A=Add(A=@n23, B=@n29), B=@n25)
  ) {equalsThreshold=0.01}
  n32 = NamedRerouteUsage()
  n33 = VertexColor()
  n34 = Constant() {r=0.6}
  n35 = VectorParameter() {parameterName=Color_FinalOverlay, defaultValue=<rgba 50 38.4612 0 0>}
  n36 = LinearInterpolate(
   A=ComponentMask(None=Constant4Vector().RGBA) {b=true, g=true, r=true}
   B=MaterialFunctionCall(R=@n35.R, G=@n35.R, B=@n35.R) {materialFunction=&AppendMany}.RGB
   Alpha=ScalarParameter() {parameterName=Overlay_Lerp}
  )
  n37 = VectorParameter() {parameterName=Color_Core, defaultValue=<rgba 0 12.894 20 0>}
  n38 = Add(A=Add(A=Multiply(A=@n2, B=@n26), B=Multiply(A=@n31, B=@n32)), B=Multiply(A=@n33.B, B=If(A=@n33.B, B=@n34, "A > B"=@n36, "A < B"=@n37.RGB)))
  n39 = CameraVectorWS()
  n40 = VertexNormalWS()
  n41 = Saturate(None=Abs(None=DotProduct(A=@n39, B=@n40)))
  n42 = Power(Base=Multiply(A=@n41, B=@n41), Exp=Constant() {r=0.8})
  n43 = NamedRerouteUsage()
  n44 = NamedRerouteUsage()
  n45 = Constant()
  n46 = Multiply(A=VertexNormalWS(), B=ScalarParameter() {defaultValue=1, parameterName=WPO_SelectedHeight})
  n47 = Multiply(A=VertexNormalWS(), B=ScalarParameter() {defaultValue=1, parameterName=WPO_AverageHeight})
  n48 = NamedRerouteUsage()
  n49 = NamedRerouteUsage()
  n50 = Constant()
  n51 = Multiply(A=VertexNormalWS(), B=ScalarParameter() {defaultValue=1, parameterName=WPO_SelectedHeight_RightSide})
  n52 = Multiply(A=VertexNormalWS(), B=ScalarParameter() {defaultValue=1, parameterName=WPO_AverageHeight_RightSide})
  n53 = Add(
   A=Multiply(A=If(A=@n43, B=@n44, "A > B"=@n45, "A == B"=@n46, "A < B"=@n47) {equalsThreshold=0.01}, B=NamedRerouteUsage())
   B=Multiply(A=If(A=@n48, B=@n49, "A > B"=@n50, "A == B"=@n51, "A < B"=@n52) {equalsThreshold=0.01}, B=NamedRerouteUsage())
  )
  MaterialFunctionCall(In=SetMaterialAttributes!(
    MaterialAttributes
    Emissive Color=NamedRerouteDeclaration(None=@n38) {name=emissive, nodeColor=<rgba 1 0.2344 0 1>}
    Opacity=NamedRerouteDeclaration(None=@n42) {name=opacity, nodeColor=<rgba 0.8047 0 1 1>}
    World Position Offset=NamedRerouteDeclaration(None=@n53) {name=wpo, nodeColor=<rgba 0.2187 1 0 1>}
   ) {attributeSetTypes [#1, #2, #3]}) {materialFunction=&MF_TechArtOverrides_accessories}
  MaterialFunctionCall(Speed=ScalarParameter() {defaultValue=1, parameterName=speed_two}) {materialFunction=&DebugTimeSine}
  n54 = VectorParameter() {defaultValue=<rgba 0 1 0.32 0>}
  n55 = Transform(None=Normalize(VectorInput=ComponentMask(None=@n54.RGB) {b=true, g=true, r=true})) {transformSourceType=TRANSFORMSOURCE_Local}
  n56 = Time()
  n57 = MaterialFunctionCall() {materialFunction=&ObjectPivotPoint}
  n58 = WorldPosition()
  Multiply(
   A=LinearInterpolate(B=RotateAboutAxis(NormalizedRotationAxis=@n55, RotationAngle=@n56, PivotPoint=@n57.Object Pivot Location, Position=@n58.XYZ))
   B=VertexColor().B
  )
  n59 = VertexColor()
  NamedRerouteDeclaration(None=Ceil(None=@n59.G)) {name=GreenVertexColor_Ceiled, nodeColor=<rgba 1 0 0.5391 1>}
  NamedRerouteDeclaration(None=@n59.G) {name=GreenVertexColor, nodeColor=<rgba 0.8047 1 0 1>}
  NamedRerouteDeclaration(None=Add(A=ScalarParameter() {defaultValue=0.97, group=Functional, parameterName=RightCounter_Progress}) {constB=0}) {name=RightCounterProgress, nodeColor=<rgba 1 0 0.8672 1>}
  ScalarParameter() {defaultValue=0.35, parameterName=SquareOutlineSize}
  VectorParameter() {parameterName=dg}
 ]
}
