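# Animation Blueprint for the Ghost Hunter backpack that nudges the pack when the wearing character is also Ghost Hunter.
# Each update casts the owner to FortPlayerPawn and wildcard-matches a part mesh's display name against the Ghost Hunter name, storing the result as a float.
# That float drives the alpha of a Modify Bone node, toggling a small additive bone-space translation on the pack bone over the input pose.
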AnimBlueprint /BRCosmetics/Accessories/FORT_Backpacks/Backpack_F_MED_Ghost_Hunter/Meshes/F_MED_Ghost_Hunter_Pack_Skeleton_AnimBlueprint
target_skeleton=&/F_MED_Ghost_Hunter_Pack_Skeleton
graphs [
 AnimGraph {nodes [
   n1 = "Transform (Modify) Bone - Bone: pack"(
    ComponentPose=Local To Component(LocalPose=Input Pose())
    Alpha=Get isGhostHunter() -> (isGhostHunter:double)
    Translation="0,2,1"
    Rotation="0,0,0"
   ) {
    rotationMode=BMM_Additive
    rotationSpace=BCS_BoneSpace
    translationMode=BMM_Additive
    translationSpace=BCS_BoneSpace
    alphaBoolBlend {blendOption=Linear}
    boneToModify {boneName=pack}
    translation {y=2, z=1}
   }
   Output Pose(Result=Component To Local(ComponentPose=@n1)) {layerGroup=None}
  ]}
 EventGraph {nodes [
   Event BlueprintUpdateAnimation()
   > n1 = Cast To FortPlayerPawn(Object=TryGetPawnOwner())
   n2 = GetDisplayName(Object=GetSkeletalMeshForPartType(self=@n1))
   > Set isGhostHunter(isGhostHunter:double="To Float (Boolean)"(InBool=MatchesWildcard(SourceString=@n2, Wildcard=*F_MED_Ghost_Hunter*))) -> (Output_Get:double)
   TryGetPawnOwner()
  ]}
]
variables [isGhostHunter {type=number}]
